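# Physics asset for the Relish backpack cosmetic, simulating a potted plant with a stem and leaves.
# One capsule body per bone covers the pot master joint, four main stem segments and nineteen leaves.
# The stem chains up from the pot with tight swing limits and locked twist, keeping it nearly rigid.
# Each leaf hangs off a stem segment with its own swing and twist limits, many asymmetric to favor one axis of motion.
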PhysicsAsset /BRCosmetics/Accessories/FORT_Backpacks/Backpack_M_MED_Relish/Meshes/M_MED_Relish_Pack_Physics
bodies {
 dyn_leaf10_jnt [{shapeType=Capsule, length=2, radius=2, shapeName=dyn_leaf10_jnt_capsule, center=<0 0.8844 2.9133>, rotation=<rot 0 0 20.0003>}]
 dyn_leaf11_jnt [{
   shapeType=Capsule
   length=4
   radius=2.75
   shapeName=dyn_leaf11_jnt_capsule
   center=<-1.642 0.3696 4.3072>
   rotation=<rot -0.0004 0.0016 20.0002>
  }]
 dyn_leaf12_jnt [{shapeType=Capsule, length=2, radius=2, shapeName=dyn_leaf12_jnt_capsule, center=<0 0.3675 2.7941>, rotation=<rot 0 0 10.0007>}]
 dyn_leaf13_jnt [{shapeType=Capsule, length=6, radius=3.25, shapeName=dyn_leaf13_jnt_capsule, center=<3.193 0.0002 5.4464>, rotation=<rot -30.0001 0 0>}]
 dyn_leaf14_jnt [{shapeType=Capsule, length=3.75, radius=3, shapeName=dyn_leaf14_jnt_capsule, center=<0 -2.9752 3.763>, rotation=<rot 0 0 -39.9998>}]
 dyn_leaf15_jnt [{
   shapeType=Capsule
   length=4.5
   radius=2.25
   shapeName=dyn_leaf15_jnt_capsule
   center=<1.3878 0.3148 4.0575>
   rotation=<rot -19.3723 -3.6096 20.9185>
  }]
 dyn_leaf16_jnt [{
   shapeType=Capsule
   length=4.25
   radius=3
   shapeName=dyn_leaf16_jnt_capsule
   center=<-1.3495 0.6609 4.7541>
   rotation=<rot 19.8417 1.8372 10.4759>
  }]
 dyn_leaf17split_jnt [{shapeType=Capsule, length=4, radius=2.75, shapeName=dyn_leaf17split_jnt_capsule, center=<0 0 4.7062>}]
 dyn_leaf18_jnt [{shapeType=Capsule, length=5.5, radius=2.75, shapeName=dyn_leaf18_jnt_capsule, center=<0.4984 -1.8959 4.8412>, rotation=<rot 0 0 -20.0016>}]
 dyn_leaf19split_jnt [{shapeType=Capsule, length=4, radius=2, shapeName=dyn_leaf19split_jnt_capsule, center=<-0.3024 0.5233 2.7304>, rotation=<rot -90 0 30.0091>}]
 dyn_leaf1_jnt [{shapeType=Capsule, length=5.5, radius=3.75, shapeName=dyn_leaf1_jnt_capsule, center=<-0.0001 -3.2906 5.9587>, rotation=<rot 0 0 -29.9994>}]
 dyn_leaf2_jnt [{shapeType=Capsule, length=2, radius=2, shapeName=dyn_leaf2_jnt_capsule, center=<0.0001 -1.2677 2.6719>, rotation=<rot 0 0 -30.0005>}]
 dyn_leaf3_jnt [{shapeType=Capsule, length=4.75, radius=2.75, shapeName=dyn_leaf3_jnt_capsule, center=<2.3313 0.5476 4.5692>, rotation=<rot -29.9998 0 0>}]
 dyn_leaf4_jnt [{shapeType=Capsule, length=3.75, radius=2.5, shapeName=dyn_leaf4_jnt_capsule, center=<-2.095 0.0001 4.0497>, rotation=<rot 29.9999 0 0>}]
 dyn_leaf5_jnt [{
   shapeType=Capsule
   length=4.5
   radius=2
   shapeName=dyn_leaf5_jnt_capsule
   center=<3.019 2.6226 1.731>
   rotation=<rot -26.9122 -34.8957 66.0698>
  }]
 dyn_leaf6split_jnt [{
   shapeType=Capsule
   length=8.25
   radius=2
   shapeName=dyn_leaf6split_jnt_capsule
   center=<0.1893 0 1.0737>
   rotation=<rot -70.3059 169.8484 169.848>
  }]
 dyn_leaf7_jnt [{shapeType=Capsule, length=6, radius=3.5, shapeName=dyn_leaf7_jnt_capsule, center=<2.0024 0.0001 6.2447>, rotation=<rot -19.9999 0 0>}]
 dyn_leaf8_jnt [{shapeType=Capsule, length=2, radius=2, shapeName=dyn_leaf8_jnt_capsule, center=<-0.4099 0.0001 2.8334>}]
 dyn_leaf9_jnt [{
   shapeType=Capsule
   length=5.5
   radius=3
   shapeName=dyn_leaf9_jnt_capsule
   center=<-2.2357 -0.5643 5.403>
   rotation=<rot 30.0002 -0.0002 -0.0011>
  }]
 dyn_mainstem_1_jnt [{shapeType=Capsule, length=7.75, radius=1, shapeName=dyn_mainstem_1_jnt_capsule, center=<0 -0.0001 4.9229>}]
 dyn_mainstem_2_jnt [{shapeType=Capsule, length=7.5, radius=1, shapeName=dyn_mainstem_2_jnt_capsule, center=<0 0 4.9962>}]
 dyn_mainstem_3_jnt [{shapeType=Capsule, length=3.5, radius=1, shapeName=dyn_mainstem_3_jnt_capsule, center=<0 0.0001 2.7786>}]
 dyn_mainstem_4_jnt [{shapeType=Capsule, length=0.5, radius=1, shapeName=dyn_mainstem_4_jnt_capsule, center=<0.0001 0.1995 1.2579>, rotation=<rot 0 0 10.0004>}]
 dyn_pot_jnt_master [{shapeType=Capsule, length=2, radius=2, shapeName=dyn_pot_jnt_master_capsule, rotation=<rot 0 -90 0>}]
}
constraints [
 {bone1Name=dyn_leaf6split_jnt, bone2Name=dyn_mainstem_1_jnt, swing1LimitDegrees=50, swing2LimitDegrees=50, twistLimitDegrees=2}
 {
  bone1Name=dyn_mainstem_2_jnt
  bone2Name=dyn_mainstem_1_jnt
  swing1LimitDegrees=2
  swing2LimitDegrees=1
  twistLimitDegrees=2
  twistMotion=Locked
 }
 {
  bone1Name=dyn_mainstem_3_jnt
  bone2Name=dyn_mainstem_2_jnt
  swing1LimitDegrees=2
  swing2LimitDegrees=1
  twistLimitDegrees=2
  twistMotion=Locked
 }
 {
  bone1Name=dyn_mainstem_4_jnt
  bone2Name=dyn_mainstem_3_jnt
  swing1LimitDegrees=2
  swing2LimitDegrees=1
  twistLimitDegrees=2
  twistMotion=Locked
 }
 {bone1Name=dyn_leaf15_jnt, bone2Name=dyn_mainstem_4_jnt, swing1LimitDegrees=15, swing2LimitDegrees=15, twistLimitDegrees=5}
 {bone1Name=dyn_leaf16_jnt, bone2Name=dyn_mainstem_4_jnt, swing1LimitDegrees=50, swing2LimitDegrees=50, twistLimitDegrees=5}
 {bone1Name=dyn_leaf17split_jnt, bone2Name=dyn_mainstem_4_jnt, swing1LimitDegrees=2, swing2LimitDegrees=2, twistLimitDegrees=50}
 {bone1Name=dyn_leaf18_jnt, bone2Name=dyn_mainstem_4_jnt, swing1LimitDegrees=50, swing2LimitDegrees=5, twistLimitDegrees=5}
 {bone1Name=dyn_leaf19split_jnt, bone2Name=dyn_mainstem_4_jnt, swing1LimitDegrees=5, swing2LimitDegrees=5, twistLimitDegrees=2}
 {bone1Name=dyn_leaf11_jnt, bone2Name=dyn_mainstem_3_jnt, swing1LimitDegrees=2, swing2LimitDegrees=2, twistLimitDegrees=60}
 {bone1Name=dyn_leaf12_jnt, bone2Name=dyn_mainstem_3_jnt, swing1LimitDegrees=15, swing2LimitDegrees=15, twistLimitDegrees=5}
 {bone1Name=dyn_leaf13_jnt, bone2Name=dyn_mainstem_3_jnt, swing1LimitDegrees=25, swing2LimitDegrees=50, twistLimitDegrees=2}
 {bone1Name=dyn_leaf14_jnt, bone2Name=dyn_mainstem_3_jnt, swing1LimitDegrees=5, swing2LimitDegrees=5, twistLimitDegrees=70}
 {bone1Name=dyn_leaf7_jnt, bone2Name=dyn_mainstem_2_jnt, swing1LimitDegrees=15, swing2LimitDegrees=35, twistLimitDegrees=5}
 {bone1Name=dyn_leaf8_jnt, bone2Name=dyn_mainstem_2_jnt, swing1LimitDegrees=15, swing2LimitDegrees=15, twistLimitDegrees=5}
 {bone1Name=dyn_leaf9_jnt, bone2Name=dyn_mainstem_2_jnt, swing1LimitDegrees=15, swing2LimitDegrees=45, twistLimitDegrees=5}
 {bone1Name=dyn_leaf10_jnt, bone2Name=dyn_mainstem_2_jnt, swing1LimitDegrees=15, swing2LimitDegrees=15, twistLimitDegrees=5}
 {bone1Name=dyn_leaf1_jnt, bone2Name=dyn_mainstem_1_jnt, swing1LimitDegrees=15, swing2LimitDegrees=15, twistLimitDegrees=50}
 {bone1Name=dyn_leaf2_jnt, bone2Name=dyn_mainstem_1_jnt, swing1LimitDegrees=15, swing2LimitDegrees=15, twistLimitDegrees=5}
 {bone1Name=dyn_leaf3_jnt, bone2Name=dyn_mainstem_1_jnt, swing1LimitDegrees=15, swing2LimitDegrees=50, twistLimitDegrees=2}
 {bone1Name=dyn_leaf4_jnt, bone2Name=dyn_mainstem_1_jnt, swing1LimitDegrees=15, swing2LimitDegrees=55, twistLimitDegrees=10}
 {bone1Name=dyn_leaf5_jnt, bone2Name=dyn_mainstem_1_jnt, swing1LimitDegrees=2, swing2LimitDegrees=2, twistLimitDegrees=50}
 {
  bone1Name=dyn_mainstem_1_jnt
  bone2Name=dyn_pot_jnt_master
  swing1LimitDegrees=2
  swing2LimitDegrees=1
  twistLimitDegrees=2
  twistMotion=Locked
 }
]
