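# Animation Blueprint for the M_MED_Relish backpack that adds physics-driven secondary motion to the pack.
# A RigidBody node simulates over the reference pose in pack-bone space, using an override physics asset, custom gravity, damped component motion and linear limits clamped to the ref pose.
# The update event derives a Skydiving float by subtracting the skydiving boolean, feeding it into the simulation alpha through range mapping and interpolation.
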AnimBlueprint /BRCosmetics/Accessories/FORT_Backpacks/Backpack_M_MED_Relish/Meshes/M_MED_Relish_Pack_AnimBP
target_skeleton=&/M_MED_Relish_Pack_Skeleton
graphs [
 AnimGraph {nodes [
   n1 = RigidBody(ComponentPose=Local To Component(LocalPose=Local Space Ref Pose()), Alpha=Get Skydiving() -> (Skydiving:double)) {
    bClampLinearTranslationLimitToRefPose=true
    bOverrideWorldGravity=true
    overridePhysicsAsset=&/M_MED_Relish_Pack_Physics
    simulationSpace=BaseBoneSpace
    alphaBoolBlend {blendOption=Linear}
    alphaScaleBiasClamp {bInterpResult=true, bMapRange=true, interpSpeedDecreasing=3, interpSpeedIncreasing=3, outRange {min=0.2}}
    baseBoneRef {boneName=pack}
    componentAppliedLinearAccClamp=<3000 3000 3000>
    componentLinearAccScale=<0.3 0.2 0.2>
    componentLinearVelScale=<0.3 0.2 0.2>
    overrideWorldGravity {z=-980}
   }
   Output Pose(Result=Component To Local(ComponentPose=@n1)) {layerGroup=None}
  ]}
 EventGraph {nodes [
   Event BlueprintUpdateAnimation()
   n1 = Get bIsSkyDiving()
   > Set Skydiving(Skydiving:double=float - float(B="To Float (Boolean)"(InBool=@n1)) {is=title K2Node_CallFunction}) -> (Output_Get:double)
   TryGetPawnOwner()
  ]}
]
variables [Skydiving {type=number}]
